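# GrillMoon_fel1000_gam is a Level Sequence that animates a single possessed TargetPoint, GrillMoon_TargetPoint, in the fel1000 sequence.
# One 3D transform track keys its location through long holds, quick relocations and a near-instant jump in Y and Z.
# Yaw turns negative in steps midway and then holds, while pitch, roll and scale stay constant.
# It has no other tracks, and nothing here shows what consumes the point.
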LevelSequence /GrillMoon/Sequence/fel/fel1000/GrillMoon_fel1000_gam
bindings [GrillMoon_TargetPoint {
  id=#1
  possessed_class=TargetPoint
  tracks [{
    type=MovieScene3DTransformTrack
    sections [{
      type=MovieScene3DTransformSection
      channels [
       Location.X {
        type=MovieSceneScriptingDoubleChannel
        default=-57592.7912
        thinned_from=29
        keys [[3600,-57592.793],[7936,-57592.793],[8064,-69683.9141],[8188,-69683.9141],[8281,-61214.3359],[8466,-74274.8125],[8755,-67475.1797],[8926,-87118.4062],[9037,-87118.4062],[9163,-61948.6875],[9316,-61948.6875],[9438,-56530.5625],[9768.2275,-57639.207],[9955,-57639.207],[10066,-78332.6719],[10369,-75990.8203],[10618,-61154.7188],[11180,-61154.7188],[11229,-22249.1797],[12214,-22249.1797],[12551,-57611.0039]]
       }
       Location.Y {
        type=MovieSceneScriptingDoubleChannel
        default=42054.3007
        thinned_from=29
        keys [[3600,41817.2578],[5683,42047.832],[5963,42885.8594],[6502,42525.957],[6809,42525.957],[6840,42103.4727],[7499,42103.4727],[7501,51292.5],[7936,51292.5],[8064,58901.5703],[8188,58901.5703],[8281,54961.7266],[8466,65598.7266],[8755,60060.8164],[8926,89929.9375],[9037,89929.9375],[9163,78005.7109],[9316,78005.7109],[9438,60374.0117],[9768.2275,42760.8711],[9955,42760.8711],[10066,66191.7734],[10369,59727.1836],[10618,63103.1289],[11180,63103.1289],[11229,45434.0625],[12214,45434.0625],[12551,42968.1758]]
       }
       Location.Z {
        type=MovieSceneScriptingDoubleChannel
        default=10132.1009
        thinned_from=29
        keys [[3600,10132.1006],[5683,10132.1006],[5963,10492.2832],[6502,10355.168],[6809,10355.168],[6840,10318.3633],[7499,10318.3633],[7501,19194.8965],[9438,19194.8965],[9768.2275,10499.2607],[9955,10499.2607],[10066,19521.5957],[12214,19521.5957],[12551,10340.791]]
       }
       Rotation.X {type=MovieSceneScriptingDoubleChannel, default=0, const [0,18]}
       Rotation.Y {type=MovieSceneScriptingDoubleChannel, default=0, const [0,18]}
       Rotation.Z {
        type=MovieSceneScriptingDoubleChannel
        default=0
        thinned_from=18
        keys [[8281,-24.9467],[8466,-39.1608],[8755,-39.1608],[8926,-52.2721],[12551,-52.2721]]
       }
       Scale.X {type=MovieSceneScriptingDoubleChannel, default=1, const [1,18]}
       Scale.Y {type=MovieSceneScriptingDoubleChannel, default=1, const [1,18]}
       Scale.Z {type=MovieSceneScriptingDoubleChannel, default=1, const [1,18]}
      ]
     }]
   }]
 }]
display_rate {denominator=1, numerator=60}
tick_resolution {denominator=1, numerator=24000}
tracks []
